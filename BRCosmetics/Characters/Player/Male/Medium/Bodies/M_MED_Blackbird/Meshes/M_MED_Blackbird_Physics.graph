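# Physics asset for the male medium Blackbird cosmetic body, covering its dangling jacket, cuff and thigh-pouch accessories.
# Capsule bodies on pelvis, spine, thighs and calf twists form a skeleton linked by free angular constraints.
# Angle-limited constraints chain O-ring to zipper to lower spine, and cuff straps to calves.
# The pouch strap, buckle and flap chain into a pouch master body, which joins the right thigh with free rotation.
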PhysicsAsset /BRCosmetics/Characters/Player/Male/Medium/Bodies/M_MED_Blackbird/Meshes/M_MED_Blackbird_Physics
bodies {
 calf_twist_01_l [{shapeType=Capsule, length=1, radius=1, shapeName=calf_twist_01_l_capsule}]
 calf_twist_01_r [{shapeType=Capsule, length=1, radius=1, shapeName=calf_twist_01_r_capsule}]
 dyn_L_cuff_strap_jnt [{
   shapeType=Capsule
   length=2.8186
   radius=1.7893
   shapeName=dyn_L_cuff_strap_jnt_capsule
   center=<3.2036 0 0>
   rotation=<rot -90 179.9802 -179.9803>
  }]
 dyn_R_cuff_strap_jnt [{
   shapeType=Capsule
   length=2.7039
   radius=1.3004
   shapeName=dyn_R_cuff_strap_jnt_capsule
   center=<-2.6877 0 0>
   rotation=<rot -90 180 179.9999>
  }]
 dyn_R_thigh_pouch_flap_jnt [{
   shapeType=Capsule
   length=7.7858
   radius=3.9798
   shapeName=dyn_R_thigh_pouch_flap_jnt_capsule
   center=<5.2897 0.47 -2.7868>
   rotation=<rot 0 0 90>
  }]
 dyn_jacket_oring_jnt [{shapeType=Capsule, length=2, radius=2, shapeName=dyn_jacket_oring_jnt_capsule, center=<2.8829 0.8272 0>, rotation=<rot -90 180 -179.9997>}]
 dyn_large_jacket_zipper_jnt [{
   shapeType=Capsule
   length=2.4847
   radius=0.9596
   shapeName=dyn_large_jacket_zipper_jnt_capsule
   center=<1.7666 0.2472 0>
   rotation=<rot -90 0 0.0004>
  }]
 dyn_pouch_buckle_jnt [{
   shapeType=Capsule
   length=1.8576
   radius=1.8134
   shapeName=dyn_pouch_buckle_jnt_capsule
   center=<0.7045 0 -0.5925>
   rotation=<rot 80 -180 179.9997>
  }]
 dyn_pouch_strap_jnt [{
   shapeType=Capsule
   length=5.6428
   radius=3.2742
   shapeName=dyn_pouch_strap_jnt_capsule
   center=<6.1693 0 -1.3713>
   rotation=<rot -90 90 -89.9992>
  }]
 master_R_thigh_pouch_jnt [{
   shapeType=Capsule
   length=7
   radius=7
   shapeName=master_R_thigh_pouch_jnt_capsule
   center=<-1.0864 0 3.4373>
   rotation=<rot -80 -179.9998 -180>
  }]
 pelvis [{shapeType=Capsule, length=11.425, radius=10.5723, shapeName=pelvis_capsule, center=<-5.0444 0.0002 0>}]
 spine_01 [{shapeType=Capsule, length=5, radius=5, shapeName=spine_01_capsule}]
 spine_04 [{shapeType=Capsule, length=5, radius=5, shapeName=spine_04_capsule}]
 thigh_l [{shapeType=Capsule, length=36.7867, radius=8.6437, shapeName=thigh_l_capsule, center=<-18.9909 1.0281 0.4298>, rotation=<rot -90 -0.0001 0>}]
 thigh_r [{shapeType=Capsule, length=2, radius=2, shapeName=thigh_r_capsule}]
}
constraints [
 {bone1Name=dyn_pouch_strap_jnt, bone2Name=dyn_pouch_buckle_jnt, swing1LimitDegrees=45, swing2LimitDegrees=50, twistLimitDegrees=2}
 {bone1Name=dyn_pouch_buckle_jnt, bone2Name=dyn_R_thigh_pouch_flap_jnt, swing1LimitDegrees=10, swing2LimitDegrees=25, twistLimitDegrees=2}
 {bone1Name=dyn_jacket_oring_jnt, bone2Name=dyn_large_jacket_zipper_jnt, swing1LimitDegrees=45, swing2LimitDegrees=45, twistLimitDegrees=2}
 {bone1Name=dyn_L_cuff_strap_jnt, bone2Name=calf_twist_01_l, swing1LimitDegrees=25, swing2LimitDegrees=15, twistLimitDegrees=2}
 {bone1Name=dyn_R_cuff_strap_jnt, bone2Name=calf_twist_01_r, swing1LimitDegrees=25, swing2LimitDegrees=15, twistLimitDegrees=2}
 {
  bone1Name=spine_04
  bone2Name=spine_01
  swing1LimitDegrees=45
  swing1Motion=Free
  swing2LimitDegrees=45
  swing2Motion=Free
  twistLimitDegrees=45
  twistMotion=Free
 }
 {bone1Name=dyn_large_jacket_zipper_jnt, bone2Name=spine_01, swing1LimitDegrees=15, swing2LimitDegrees=15, twistLimitDegrees=2}
 {
  bone1Name=calf_twist_01_r
  bone2Name=thigh_r
  swing1LimitDegrees=45
  swing1Motion=Free
  swing2LimitDegrees=45
  swing2Motion=Free
  twistLimitDegrees=45
  twistMotion=Free
 }
 {
  bone1Name=spine_01
  bone2Name=pelvis
  swing1LimitDegrees=45
  swing1Motion=Free
  swing2LimitDegrees=45
  swing2Motion=Free
  twistLimitDegrees=45
  twistMotion=Free
 }
 {
  bone1Name=calf_twist_01_l
  bone2Name=thigh_l
  swing1LimitDegrees=45
  swing1Motion=Free
  swing2LimitDegrees=45
  swing2Motion=Free
  twistLimitDegrees=45
  twistMotion=Free
 }
 {
  bone1Name=thigh_r
  bone2Name=pelvis
  swing1LimitDegrees=45
  swing1Motion=Free
  swing2LimitDegrees=45
  swing2Motion=Free
  twistLimitDegrees=45
  twistMotion=Free
 }
 {
  bone1Name=thigh_l
  bone2Name=pelvis
  swing1LimitDegrees=45
  swing1Motion=Free
  swing2LimitDegrees=45
  swing2Motion=Free
  twistLimitDegrees=45
  twistMotion=Free
 }
 {bone1Name=dyn_R_thigh_pouch_flap_jnt, bone2Name=master_R_thigh_pouch_jnt, swing1LimitDegrees=2, swing2LimitDegrees=2, twistLimitDegrees=45}
 {
  bone1Name=master_R_thigh_pouch_jnt
  bone2Name=thigh_r
  swing1LimitDegrees=45
  swing1Motion=Free
  swing2LimitDegrees=45
  swing2Motion=Free
  twistLimitDegrees=45
  twistMotion=Free
 }
]
