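# Animation Blueprint for the M_MED_Blackbird player body cosmetic, adding secondary motion and corrective posing to its outfit.
# It copies the attached parent mesh's pose, then runs a RigidBody simulation on the outfit's physics asset.
# Chained weighted multi-bone constraints place jacket bolts, zipper and D-ring masters, then spring-driven AnimDynamics swings both D-rings.
# Additive bone offsets finally adjust the zipper when crouching, with interpolated blending, and three bolts when targeting.
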AnimBlueprint /BRCosmetics/Characters/Player/Male/Medium/Bodies/M_MED_Blackbird/Meshes/M_MED_Blackbird_AnimBP
target_skeleton=&/M_MED_Blackbird_Skeleton
graphs [
 AnimGraph {nodes [
   n1 = RigidBody(ComponentPose=Local To Component(LocalPose=Copy Pose From Mesh() {bUseAttachedParent=true})) {
    bClampLinearTranslationLimitToRefPose=true
    bOverrideWorldGravity=true
    overridePhysicsAsset=&/M_MED_Blackbird_Physics
    simulationSpace=BaseBoneSpace
    alphaBoolBlend {blendOption=Linear}
    baseBoneRef {boneName=root}
    componentAppliedLinearAccClamp=<3000 3000 3000>
    componentLinearAccScale=<0.2 0.1 0.1>
    componentLinearVelScale=<0.2 0.1 0.1>
    overrideWorldGravity {z=-980}
   }
   n2 = "Constraint - R_shoulder_bolt_jnt (PPPPPP)"!(
    ComponentPose=@n1
    Alpha=1
    ConstraintWeights_0=0
    ConstraintWeights_1:float=0.0183
    ConstraintWeights_2:float=0.0756
    ConstraintWeights_3:float=0.0114
    ConstraintWeights_4:float=0.5833
    ConstraintWeights_5:float=0.3115
   ) -> (Pose) {
    alphaBoolBlend {blendOption=Linear}
    boneToModify {boneName=R_shoulder_bolt_jnt}
    constraintSetup [
     {targetBone {boneName=spine_02}}
     {targetBone {boneName=spine_03}}
     {targetBone {boneName=spine_05}}
     {targetBone {boneName=neck_02}}
     {targetBone {boneName=clavicle_l}}
     {targetBone {boneName=upperarm_twist_01_l}}
    ]
    constraintWeights [1,1,1,1,1,1]
   }
   n3 = "Constraint - L_back_lower_bolt_jnt (PPPPPP)"!(
    ComponentPose=@n2
    Alpha=1
    ConstraintWeights_0=0.0912
    ConstraintWeights_1:float=0.0503
    ConstraintWeights_2:float=0.2236
    ConstraintWeights_3:float=0.0236
    ConstraintWeights_4:float=0.4494
    ConstraintWeights_5:float=0.162
   ) -> (Pose) {
    alphaBoolBlend {blendOption=Linear}
    boneToModify {boneName=L_back_lower_bolt_jnt}
    constraintSetup [
     {targetBone {boneName=spine_03}}
     {targetBone {boneName=spine_04}}
     {targetBone {boneName=spine_05}}
     {targetBone {boneName=neck_02}}
     {targetBone {boneName=clavicle_l}}
     {targetBone {boneName=upperarm_twist_01_l}}
    ]
    constraintWeights [1,1,1,1,1,1]
   }
   n4 = "Constraint - L_front_sleeve_bolt_jnt (PPPPPPPP)"!(
    ComponentPose=@n3
    Alpha=1
    ConstraintWeights_0=0
    ConstraintWeights_1:float=0.0133
    ConstraintWeights_2:float=0.0131
    ConstraintWeights_3:float=0.0217
    ConstraintWeights_4:float=0.3343
    ConstraintWeights_5:float=0.0286
    ConstraintWeights_6:float=0.5624
    ConstraintWeights_7:float=0.0265
   ) -> (Pose) {
    alphaBoolBlend {blendOption=Linear}
    boneToModify {boneName=L_front_sleeve_bolt_jnt}
    constraintSetup [
     {targetBone {boneName=spine_02}}
     {targetBone {boneName=spine_03}}
     {targetBone {boneName=spine_04}}
     {targetBone {boneName=spine_05}}
     {targetBone {boneName=clavicle_l}}
     {targetBone {boneName=upperarm_l}}
     {targetBone {boneName=upperarm_twist_01_l}}
     {targetBone {boneName=upperarm_twist_02_l}}
    ]
    constraintWeights [1,1,1,1,1,1,1,1]
   }
   n5 = "Constraint - L_front_lower_bolt_jnt (PPPPPPP)"!(
    ComponentPose=@n4
    Alpha=1
    ConstraintWeights_0=0.0462
    ConstraintWeights_1:float=0.0695
    ConstraintWeights_2:float=0.2696
    ConstraintWeights_3:float=0.0135
    ConstraintWeights_4:float=0.0445
    ConstraintWeights_5:float=0.379
    ConstraintWeights_6:float=0.1777
   ) -> (Pose) {
    alphaBoolBlend {blendOption=Linear}
    boneToModify {boneName=L_front_lower_bolt_jnt}
    constraintSetup [
     {targetBone {boneName=spine_03}}
     {targetBone {boneName=spine_04}}
     {targetBone {boneName=spine_05}}
     {targetBone {boneName=neck_01}}
     {targetBone {boneName=neck_02}}
     {targetBone {boneName=clavicle_l}}
     {targetBone {boneName=upperarm_twist_01_l}}
    ]
    constraintWeights [1,1,1,1,1,1,1]
   }
   n6 = "Constraint - L_front_upper_bolt_jnt (PPPPPPPP)"!(
    ComponentPose=@n5
    Alpha=1
    ConstraintWeights_0=0.0183
    ConstraintWeights_1:float=0.0293
    ConstraintWeights_2:float=0.3977
    ConstraintWeights_3:float=0.0445
    ConstraintWeights_4:float=0.1172
    ConstraintWeights_5:float=0.0213
    ConstraintWeights_6:float=0.3244
    ConstraintWeights_7:float=0.0475
   ) -> (Pose) {
    alphaBoolBlend {blendOption=Linear}
    boneToModify {boneName=L_front_upper_bolt_jnt}
    constraintSetup [
     {targetBone {boneName=spine_03}}
     {targetBone {boneName=spine_04}}
     {targetBone {boneName=spine_05}}
     {targetBone {boneName=neck_01}}
     {targetBone {boneName=neck_02}}
     {targetBone {boneName=head}}
     {targetBone {boneName=clavicle_l}}
     {targetBone {boneName=upperarm_twist_01_l}}
    ]
    constraintWeights [1,1,1,1,1,1,1,1]
   }
   n7 = "Constraint - R_back_lower_bolt_jnt (PPPPPP)"!(
    ComponentPose=@n6
    Alpha=1
    ConstraintWeights_0=0.0493
    ConstraintWeights_1:float=0.0296
    ConstraintWeights_2:float=0.1173
    ConstraintWeights_3:float=0.0135
    ConstraintWeights_4:float=0.6927
    ConstraintWeights_5:float=0.0975
   ) -> (Pose) {
    alphaBoolBlend {blendOption=Linear}
    boneToModify {boneName=R_back_lower_bolt_jnt}
    constraintSetup [
     {targetBone {boneName=spine_03}}
     {targetBone {boneName=spine_04}}
     {targetBone {boneName=spine_05}}
     {targetBone {boneName=neck_02}}
     {targetBone {boneName=clavicle_r}}
     {targetBone {boneName=upperarm_twist_01_r}}
    ]
    constraintWeights [1,1,1,1,1,1]
   }
   n8 = "Constraint - R_back_upper_bolt_jnt (PP)"!(ComponentPose=@n7, Alpha=1, ConstraintWeights_0=0.3959, ConstraintWeights_1:float=0.6041) -> (Pose) {
    alphaBoolBlend {blendOption=Linear}
    boneToModify {boneName=R_back_upper_bolt_jnt}
    constraintSetup [{targetBone {boneName=spine_05}}, {targetBone {boneName=clavicle_r}}]
    constraintWeights [1,1]
   }
   n9 = "Constraint - L_shoulder_bolt_jnt (PPP)"!(ComponentPose=@n8, Alpha=1, ConstraintWeights_0=0.0561, ConstraintWeights_1:float=0.825, ConstraintWeights_2:float=0.1189) -> (Pose) {
    alphaBoolBlend {blendOption=Linear}
    boneToModify {boneName=L_shoulder_bolt_jnt}
    constraintSetup [{targetBone {boneName=spine_05}}, {targetBone {boneName=clavicle_r}}, {targetBone {boneName=upperarm_twist_01_r}}]
    constraintWeights [1,1,1]
   }
   n10 = "Constraint - R_front_upper_bolt_jnt (PPPPPPPP)"!(
    ComponentPose=@n9
    Alpha=1
    ConstraintWeights_0=0.017
    ConstraintWeights_1:float=0.0435
    ConstraintWeights_2:float=0.4339
    ConstraintWeights_3:float=0.0468
    ConstraintWeights_4:float=0.1204
    ConstraintWeights_5:float=0.0167
    ConstraintWeights_6:float=0.2846
    ConstraintWeights_7:float=0.0372
   ) -> (Pose) {
    alphaBoolBlend {blendOption=Linear}
    boneToModify {boneName=R_front_upper_bolt_jnt}
    constraintSetup [
     {targetBone {boneName=spine_03}}
     {targetBone {boneName=spine_04}}
     {targetBone {boneName=spine_05}}
     {targetBone {boneName=neck_01}}
     {targetBone {boneName=neck_02}}
     {targetBone {boneName=head}}
     {targetBone {boneName=clavicle_r}}
     {targetBone {boneName=upperarm_twist_01_r}}
    ]
    constraintWeights [1,1,1,1,1,1,1,1]
   }
   n11 = "Constraint - R_front_sleeve_bolt_jnt (PPPP)"!(
    ComponentPose=@n10
    Alpha=1
    ConstraintWeights_0=0.0217
    ConstraintWeights_1:float=0.0157
    ConstraintWeights_2:float=0.3206
    ConstraintWeights_3:float=0.642
   ) -> (Pose) {
    alphaBoolBlend {blendOption=Linear}
    boneToModify {boneName=R_front_sleeve_bolt_jnt}
    constraintSetup [
     {targetBone {boneName=spine_03}}
     {targetBone {boneName=spine_05}}
     {targetBone {boneName=clavicle_r}}
     {targetBone {boneName=upperarm_twist_01_r}}
    ]
    constraintWeights [1,1,1,1]
   }
   n12 = "Constraint - R_back_upperarm_bolt (PPPPPPPP)"!(
    ComponentPose=@n11
    Alpha=1
    ConstraintWeights_0=0.0143
    ConstraintWeights_1:float=0.0016
    ConstraintWeights_2:float=0.0077
    ConstraintWeights_3:float=0
    ConstraintWeights_4:float=0.3071
    ConstraintWeights_5:float=0.0364
    ConstraintWeights_6:float=0.4493
    ConstraintWeights_7:float=0.1835
   ) -> (Pose) {
    alphaBoolBlend {blendOption=Linear}
    boneToModify {boneName=R_back_upperarm_bolt}
    constraintSetup [
     {targetBone {boneName=spine_03}}
     {targetBone {boneName=spine_04}}
     {targetBone {boneName=spine_05}}
     {targetBone {boneName=neck_02}}
     {targetBone {boneName=clavicle_r}}
     {targetBone {boneName=upperarm_r}}
     {targetBone {boneName=upperarm_twist_01_r}}
     {targetBone {boneName=upperarm_twist_02_r}}
    ]
    constraintWeights [1,1,1,1,1,1,1,1]
   }
   n13 = "Constraint - R_front_lower_bolt_jnt (PPPPPP)"!(
    ComponentPose=@n12
    Alpha=1
    ConstraintWeights_0=0.0482
    ConstraintWeights_1:float=0.0935
    ConstraintWeights_2:float=0.2364
    ConstraintWeights_3:float=0.0343
    ConstraintWeights_4:float=0.4439
    ConstraintWeights_5:float=0.1437
   ) -> (Pose) {
    alphaBoolBlend {blendOption=Linear}
    boneToModify {boneName=R_front_lower_bolt_jnt}
    constraintSetup [
     {targetBone {boneName=spine_03}}
     {targetBone {boneName=spine_04}}
     {targetBone {boneName=spine_05}}
     {targetBone {boneName=neck_02}}
     {targetBone {boneName=clavicle_r}}
     {targetBone {boneName=upperarm_twist_01_r}}
    ]
    constraintWeights [1,1,1,1,1,1]
   }
   n14 = "Constraint - master_large_jacket_zipper_jnt (PPP)"!(ComponentPose=@n13, Alpha=1, ConstraintWeights_0=0.8083, ConstraintWeights_1:float=0.1454, ConstraintWeights_2:float=0.0463) -> (Pose) {
    alphaBoolBlend {blendOption=Linear}
    boneToModify {boneName=master_large_jacket_zipper_jnt}
    constraintSetup [{targetBone {boneName=pelvis}}, {targetBone {boneName=spine_01}}, {targetBone {boneName=thigh_twist_01_l}}]
    constraintWeights [1,1,1]
   }
   n15 = "Constraint - master_R_dring_jnt (PPPPPPPP)"!(
    ComponentPose=@n14
    Alpha=1
    ConstraintWeights_0=0.0195
    ConstraintWeights_1:float=0.1987
    ConstraintWeights_2:float=0.3749
    ConstraintWeights_3:float=0.3214
    ConstraintWeights_4:float=0.0854
    ConstraintWeights_5:float=0.0001
    ConstraintWeights_6:float=0
    ConstraintWeights_7:float=0.0002
   ) -> (Pose) {
    alphaBoolBlend {blendOption=Linear}
    boneToModify {boneName=master_R_dring_jnt}
    constraintSetup [
     {targetBone {boneName=spine_01}}
     {targetBone {boneName=spine_02}}
     {targetBone {boneName=spine_03}}
     {targetBone {boneName=spine_04}}
     {targetBone {boneName=spine_05}}
     {targetBone {boneName=thigh_twist_01_r}}
     {targetBone {boneName=clavicle_l}}
     {targetBone {boneName=clavicle_r}}
    ]
    constraintWeights [1,1,1,1,1,1,1,1]
   }
   n16 = "Constraint - master_L_dring_jnt (PPPPPPPP)"!(
    ComponentPose=@n15
    Alpha=1
    ConstraintWeights_0=0.0218
    ConstraintWeights_1:float=0.2053
    ConstraintWeights_2:float=0.3762
    ConstraintWeights_3:float=0.3157
    ConstraintWeights_4:float=0.0809
    ConstraintWeights_5:float=0
    ConstraintWeights_6:float=0.0001
    ConstraintWeights_7:float=0
   ) -> (Pose) {
    alphaBoolBlend {blendOption=Linear}
    boneToModify {boneName=master_L_dring_jnt}
    constraintSetup [
     {targetBone {boneName=spine_01}}
     {targetBone {boneName=spine_02}}
     {targetBone {boneName=spine_03}}
     {targetBone {boneName=spine_04}}
     {targetBone {boneName=spine_05}}
     {targetBone {boneName=thigh_twist_01_r}}
     {targetBone {boneName=clavicle_l}}
     {targetBone {boneName=clavicle_r}}
    ]
    constraintWeights [1,1,1,1,1,1,1,1]
   }
   n17 = Anim Dynamics - Bone: dyn_L_dring_jnt(ComponentPose=@n16, Alpha=1, ExternalForce="0,0,0") {
    angularSpringConstant=20000
    bAngularSpring=true
    bUseGravityOverride=true
    simulationSpace=RootRelative
    alphaBoolBlend {blendOption=Linear}
    boundBone {boneName=dyn_L_dring_jnt}
    componentAppliedLinearAccClamp=<3000 3000 3000>
    componentLinearAccScale=<0.3 0.3 0.3>
    componentLinearVelScale=<0.3 0.3 0.3>
    gravityOverride {z=-980}
    physicsBodyDefinitions [{
      boundBone {boneName=dyn_L_dring_jnt}
      boxExtents=<10 10 10>
      constraintSetup {angularLimitsMax=<0 0 5>, angularLimitsMin=<0 0 -60>, angularTarget=<0 0 0>}
      localJointOffset=<5 0 0>
     }]
    relativeSpaceBone {boneName=master_L_dring_jnt}
   }
   n18 = Anim Dynamics - Bone: dyn_R_dring_jnt(ComponentPose=@n17, Alpha=1, ExternalForce="0,0,0") {
    angularSpringConstant=20000
    bAngularSpring=true
    bUseGravityOverride=true
    simulationSpace=RootRelative
    alphaBoolBlend {blendOption=Linear}
    boundBone {boneName=dyn_R_dring_jnt}
    componentAppliedLinearAccClamp=<3000 3000 3000>
    componentLinearAccScale=<0.4 0.4 0.4>
    componentLinearVelScale=<0.4 0.4 0.4>
    gravityOverride {z=-980}
    physicsBodyDefinitions [{
      boundBone {boneName=dyn_R_dring_jnt}
      boxExtents=<10 10 10>
      constraintSetup {angularLimitsMax=<0 0 5>, angularLimitsMin=<0 0 -60>, angularTarget=<0 0 0>}
      localJointOffset=<5 0 0>
     }]
    relativeSpaceBone {boneName=master_L_dring_jnt}
   }
   n19 = "Transform (Modify) Bone - Bone: master_large_jacket_zipper_jnt"(ComponentPose=@n18, Alpha="To Float (Boolean)"(InBool=Get bIsCrouching()), Translation="0,0,0", Rotation="-40,0,0", Scale="1,1,1") {
    rotationMode=BMM_Additive
    rotationSpace=BCS_BoneSpace
    alphaBoolBlend {blendOption=Linear}
    alphaScaleBiasClamp {bInterpResult=true, interpSpeedDecreasing=2, interpSpeedIncreasing=6}
    boneToModify {boneName=master_large_jacket_zipper_jnt}
    rotation {pitch=-40}
   }
   n20 = Get bIsTargeting()
   n21 = "Transform (Modify) Bone - Bone: R_shoulder_bolt_jnt"!(ComponentPose=@n19, bAlphaBoolEnabled:bool=@n20, Translation="0,0,0", Rotation="10,0,0", Scale="1,1,1") -> (Pose) {
    alphaInputType=Bool
    rotationMode=BMM_Additive
    rotationSpace=BCS_BoneSpace
    alphaBoolBlend {blendOption=Linear}
    boneToModify {boneName=R_shoulder_bolt_jnt}
    rotation {pitch=10}
   }
   n22 = "Transform (Modify) Bone - Bone: L_front_sleeve_bolt_jnt"!(ComponentPose=@n21, bAlphaBoolEnabled:bool=@n20, Translation="0,0,-0.5", Rotation="0,0,10", Scale="1,1,1") -> (Pose) {
    alphaInputType=Bool
    rotationMode=BMM_Additive
    rotationSpace=BCS_BoneSpace
    translationMode=BMM_Additive
    translationSpace=BCS_BoneSpace
    alphaBoolBlend {blendOption=Linear}
    boneToModify {boneName=L_front_sleeve_bolt_jnt}
    rotation {roll=10}
    translation {z=-0.5}
   }
   n23 = "Transform (Modify) Bone - Bone: R_back_upperarm_bolt"!(ComponentPose=@n22, bAlphaBoolEnabled:bool=@n20, Translation="0,0,0", Rotation="0,0,-10", Scale="1,1,1") -> (Pose) {
    alphaInputType=Bool
    rotationMode=BMM_Additive
    rotationSpace=BCS_BoneSpace
    alphaBoolBlend {blendOption=Linear}
    boneToModify {boneName=R_back_upperarm_bolt}
    rotation {roll=-10}
   }
   Output Pose(Result=Component To Local(ComponentPose=@n23)) {layerGroup=None}
  ]}
 EventGraph {nodes [Event BlueprintUpdateAnimation()]}
]
variables []
